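# Physics asset for the large male TwinkleLoop head cosmetic, defining capsule bodies and joints for its dynamic beard and ears.
# Left and right three-bone beard and ear chains hang from neck_01, alongside capsules on spine_05, both neck bones, and the right forearm and hand.
# Beard roots are tightly limited with looser lower joints, while ear roots and arm joints rotate freely and outer ear joints have moderate limits.
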PhysicsAsset /BRCosmetics/Characters/Player/Male/Large/Heads/M_LRG_TwinkleLoop_Head/Meshes/M_LRG_TwinkleLoop_Head_Physics
bodies {
 dyn_beard_l_01 [{shapeType=Capsule, length=0.505, radius=0.505, shapeName=dyn_beard_l_01_capsule, rotation=<rot 0 -90 0>}]
 dyn_beard_l_02 [{
   shapeType=Capsule
   length=0.8657
   radius=3.7056
   shapeName=dyn_beard_l_02_capsule
   center=<1.9276 2.159 0.0246>
   rotation=<rot 0 -89.9999 -90.7308>
  }]
 dyn_beard_l_03 [{
   shapeType=Capsule
   length=4.5726
   radius=3.2256
   shapeName=dyn_beard_l_03_capsule
   center=<4.5223 2.2175 -0.252>
   rotation=<rot -2.8206 -89.9999 -85.4387>
  }]
 dyn_beard_r_01 [{shapeType=Capsule, length=0.505, radius=0.505, shapeName=dyn_beard_r_01_capsule, rotation=<rot 0 -90 0>}]
 dyn_beard_r_02 [{
   shapeType=Capsule
   length=1.3657
   radius=3.2056
   shapeName=dyn_beard_r_02_capsule
   center=<2.0278 0 0.0259>
   rotation=<rot 0 -89.9999 -90.7308>
  }]
 dyn_beard_r_03 [{
   shapeType=Capsule
   length=4.5726
   radius=3.2256
   shapeName=dyn_beard_r_03_capsule
   center=<3.1364 0.2408 0.2018>
   rotation=<rot 2.885 -89.9999 -103.94>
  }]
 dyn_ear_l_01 [{shapeType=Capsule, length=0.505, radius=0.505, shapeName=dyn_ear_l_01_capsule, rotation=<rot 0 -90 0>}]
 dyn_ear_l_02 [{
   shapeType=Capsule
   length=2.3206
   radius=2.578
   shapeName=dyn_ear_l_02_capsule
   center=<2.7401 0.3971 0.3474>
   rotation=<rot 0.6215 87.9892 -90.0001>
  }]
 dyn_ear_l_03 [{
   shapeType=Capsule
   length=2.3206
   radius=2.578
   shapeName=dyn_ear_l_03_capsule
   center=<4.9631 0.4233 0.1573>
   rotation=<rot 10.1217 92.9013 -90.0003>
  }]
 dyn_ear_r_01 [{shapeType=Capsule, length=0.505, radius=0.505, shapeName=dyn_ear_r_01_capsule, rotation=<rot 0 -90 0>}]
 dyn_ear_r_02 [{
   shapeType=Capsule
   length=2.3206
   radius=2.578
   shapeName=dyn_ear_r_02_capsule
   center=<2.4882 0.4057 -0.5369>
   rotation=<rot 0.8716 -91.9157 -90.0001>
  }]
 dyn_ear_r_03 [{
   shapeType=Capsule
   length=2.3206
   radius=2.578
   shapeName=dyn_ear_r_03_capsule
   center=<5.0275 0.4179 0.5814>
   rotation=<rot 10.4503 -88.7375 -90.0003>
  }]
 hand_r [{shapeType=Capsule, length=9, radius=6, shapeName=hand_r_capsule, center=<-10.6935 -1.3482 -0.5164>, rotation=<rot -0.0001 -90 90.158>}]
 lowerarm_r [{shapeType=Capsule, length=25.5, radius=4, shapeName=lowerarm_r_capsule, center=<-3.1031 0 -7.0994>, rotation=<rot 0 -90 87.4722>}]
 neck_01 [{shapeType=Capsule, length=2.0998, radius=18.7317, center=<-1.0618 0.5578 0>, rotation=<rot 0 -108.3678 0>}]
 neck_02 [{shapeType=Capsule, length=3.1775, radius=17.3213, center=<0.5222 -0.5265 0>, rotation=<rot -0.0001 -106.7609 0>}]
 spine_05 [{shapeType=Capsule, length=1.2167, radius=20.1372, center=<2.1437 0.9033 0>, rotation=<rot 0.0001 -90 0>}]
}
constraints [
 {bone1Name=neck_01, bone2Name=spine_05, swing1LimitDegrees=45, swing2LimitDegrees=45, twistLimitDegrees=45}
 {bone1Name=neck_02, bone2Name=neck_01, swing1LimitDegrees=45, swing2LimitDegrees=45, twistLimitDegrees=45}
 {bone1Name=dyn_beard_l_01, bone2Name=neck_01, swing1LimitDegrees=10, swing2LimitDegrees=10, twistLimitDegrees=1}
 {bone1Name=dyn_beard_l_02, bone2Name=dyn_beard_l_01, swing1LimitDegrees=60, swing2LimitDegrees=45, twistLimitDegrees=10}
 {bone1Name=dyn_beard_l_03, bone2Name=dyn_beard_l_02, swing1LimitDegrees=60, swing2LimitDegrees=45, twistLimitDegrees=10}
 {bone1Name=dyn_beard_r_01, bone2Name=neck_01, swing1LimitDegrees=10, swing2LimitDegrees=10, twistLimitDegrees=1}
 {bone1Name=dyn_beard_r_02, bone2Name=dyn_beard_r_01, swing1LimitDegrees=60, swing2LimitDegrees=45, twistLimitDegrees=10}
 {bone1Name=dyn_beard_r_03, bone2Name=dyn_beard_r_02, swing1LimitDegrees=60, swing2LimitDegrees=45, twistLimitDegrees=10}
 {
  bone1Name=dyn_ear_l_01
  bone2Name=neck_01
  swing1LimitDegrees=45
  swing1Motion=Free
  swing2LimitDegrees=45
  swing2Motion=Free
  twistLimitDegrees=45
  twistMotion=Free
 }
 {bone1Name=dyn_ear_l_02, bone2Name=dyn_ear_l_01, swing1LimitDegrees=20, swing2LimitDegrees=20, twistLimitDegrees=10}
 {bone1Name=dyn_ear_l_03, bone2Name=dyn_ear_l_02, swing1LimitDegrees=20, swing2LimitDegrees=20, twistLimitDegrees=10}
 {
  bone1Name=dyn_ear_r_01
  bone2Name=neck_01
  swing1LimitDegrees=45
  swing1Motion=Free
  swing2LimitDegrees=45
  swing2Motion=Free
  twistLimitDegrees=45
  twistMotion=Free
 }
 {bone1Name=dyn_ear_r_02, bone2Name=dyn_ear_r_01, swing1LimitDegrees=20, swing2LimitDegrees=20, twistLimitDegrees=10}
 {bone1Name=dyn_ear_r_03, bone2Name=dyn_ear_r_02, swing1LimitDegrees=20, swing2LimitDegrees=20, twistLimitDegrees=10}
 {
  bone1Name=lowerarm_r
  bone2Name=spine_05
  swing1LimitDegrees=45
  swing1Motion=Free
  swing2LimitDegrees=45
  swing2Motion=Free
  twistLimitDegrees=45
  twistMotion=Free
 }
 {
  bone1Name=hand_r
  bone2Name=lowerarm_r
  swing1LimitDegrees=45
  swing1Motion=Free
  swing2LimitDegrees=45
  swing2Motion=Free
  twistLimitDegrees=45
  twistMotion=Free
 }
]
