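# Post-process AnimBP for the TwinkleLoop large male head, adding procedural adjustments and physics to the character part, including its beard.
# A Control Rig pose is cached, with additive spine and clavicle offsets blended in while targeting.
# Weighted constraints tie beard root bones to spine and neck before RigidBody simulation in spine_02 space.
# A deprecated state function updates per-state gravity for the simulation; its RBAN settings output is stored but not wired in.
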AnimBlueprint /BRCosmetics/Characters/Player/Male/Large/Heads/M_LRG_TwinkleLoop_Head/Meshes/M_LRG_TwinkleLoop_Head_PostProcess_AnimBP01
target_skeleton=&/Fortnite_M_LRG_TwinkleLoop_Head_Skeleton
graphs [
 AnimGraph {nodes [
   n1 = "Transform (Modify) Bone - Bone: spine_05"!(ComponentPose=Local To Component(LocalPose=Use cached pose 'TargetPose'()), Alpha=1, Translation="-2,2,0", Rotation="0,0,0") -> (Pose) {
    rotationMode=BMM_Additive
    translationMode=BMM_Additive
    alphaBoolBlend {blendOption=Linear}
    alphaScaleBiasClamp {bInterpResult=true, interpSpeedDecreasing=5, interpSpeedIncreasing=5}
    boneToModify {boneName=spine_05}
    translation=<-2 2>
   }
   n2 = "Transform (Modify) Bone - Bone: spine_04"!(ComponentPose=@n1, Alpha=1, Translation="-1,2,0", Rotation="0,0,0") -> (Pose) {
    rotationMode=BMM_Additive
    translationMode=BMM_Additive
    alphaBoolBlend {blendOption=Linear}
    alphaScaleBiasClamp {bInterpResult=true, interpSpeedDecreasing=5, interpSpeedIncreasing=5}
    boneToModify {boneName=spine_04}
    translation=<-1 2>
   }
   n3 = "Transform (Modify) Bone - Bone: clavicle_l"!(ComponentPose=@n2, Alpha=1, Translation="3,-4,0", Rotation="0,0,0") -> (Pose) {
    rotationMode=BMM_Additive
    translationMode=BMM_Additive
    alphaBoolBlend {blendOption=Linear}
    alphaScaleBiasClamp {bInterpResult=true, interpSpeedDecreasing=5, interpSpeedIncreasing=5}
    boneToModify {boneName=clavicle_l}
    translation=<3 -4>
   }
   n4 = "Transform (Modify) Bone - Bone: clavicle_r"!(ComponentPose=@n3, Alpha=1, Translation="3,-4,0", Rotation="0,0,0") -> (Pose) {
    rotationMode=BMM_Additive
    translationMode=BMM_Additive
    alphaBoolBlend {blendOption=Linear}
    alphaScaleBiasClamp {bInterpResult=true, interpSpeedDecreasing=5, interpSpeedIncreasing=5}
    boneToModify {boneName=clavicle_r}
    translation=<3 -4>
   }
   n5 = Component To Local(ComponentPose=@n4)
   n6 = Use cached pose 'TargetPose'()
   n7 = Get bIsTargeting()
   n8 = "Constraint - dyn_beard_l_01 (PP)"!(
    ComponentPose=Local To Component(LocalPose=Blend Poses by bool(BlendPose_0=@n5, BlendPose_1=@n6, bActiveValue=@n7) {blendType=Linear})
    bAlphaBoolEnabled:bool=True
    ConstraintWeights_0=0.3
    ConstraintWeights_1:float=0.7
   ) -> (Pose) {
    alphaInputType=Bool
    alphaBoolBlend {blendOption=Linear}
    boneToModify {boneName=dyn_beard_l_01}
    constraintSetup [{targetBone {boneName=spine_05}}, {targetBone {boneName=neck_02}}]
    constraintWeights [1,1]
   }
   n9 = "Constraint - dyn_beard_r_01 (PP)"!(ComponentPose=@n8, bAlphaBoolEnabled:bool=True, ConstraintWeights_0=0.3, ConstraintWeights_1:float=0.7) -> (Pose) {
    alphaInputType=Bool
    alphaBoolBlend {blendOption=Linear}
    boneToModify {boneName=dyn_beard_r_01}
    constraintSetup [{targetBone {boneName=spine_05}}, {targetBone {boneName=neck_02}}]
    constraintWeights [1,1]
   }
   n10 = RigidBody!(
    ComponentPose=@n9
    Alpha
    OverrideWorldGravity:Vector=Get Gravity Override Cape() -> (Gravity Override Cape:Vector)
    ExternalForce
    ComponentLinearAccScale:Vector="0.5,0.3,0.3"
    ComponentLinearVelScale:Vector="0.5,0.3,0.3"
    SimSpaceSettings:Sim Space Settings Structure=(WorldAlpha=1,VelocityScaleZ=1,DampingAlpha=1,MaxLinearVelocity=3000,MaxAngularVelocity=5,MaxLinearAcceleration=3000,MaxAngularAcceleration=3,ExternalLinearDragV=(X=0,Y=0,Z=0),ExternalLinearVelocity=(X=0,Y=0,Z=0),ExternalAngularVelocity=(X=0,Y=0,Z=0))
   ) -> (Pose) {
    bClampLinearTranslationLimitToRefPose=true
    bOverrideWorldGravity=true
    overridePhysicsAsset=&/M_LRG_TwinkleLoop_Head_Physics
    simulationSpace=BaseBoneSpace
    alphaBoolBlend {blendOption=Linear}
    baseBoneRef {boneName=spine_02}
    overrideWorldGravity {z=-980}
    simSpaceSettings {maxAngularAcceleration=3, maxAngularVelocity=5, maxLinearAcceleration=3000, maxLinearVelocity=3000, worldAlpha=1}
   }
   Output Pose(Result=Component To Local(ComponentPose=@n10))
   n11 = Control Rig(Source=Input Pose()) {
    alphaBoolBlend {blendOption=Linear}
    controlRigAssetReference {blueprintRigClass=&/M_LRG_TwinkleLoop_Head_CtrlRig.M_LRG_TwinkleLoop_Head_CtrlRig_C}
   }
   Save cached pose 'TargetPose'(Pose=@n11) {cacheName=TargetPose}
  ]}
 EventGraph {nodes [
   Event BlueprintUpdateAnimation()
   > n1 = Gravity Override State Function DEPRECATED(
    CharacterPartAnimInstance=Self-Reference()
    GameTimeInSeconds=GetGameTimeInSeconds()
    GravityOverrideParams=Get Gravity Override Params()
    RBAN Params=Get RBAN Params()
   )
   > Set Gravity Override Cape(Gravity Override Cape:Vector=@n1) -> (Output_Get:Vector)
   > Set RBAN Settings Out(RBAN Settings Out:RBANSetting Struct Structure=@n1.RBAN SettingsOut) -> (Output_Get:RBANSetting Struct Structure)
   TryGetPawnOwner()
  ]}
]
variables [
 "FInterp(Alpha, (5: 5))" {type=number, default=1}
 Gravity Override Cape {type=Vector}
 Gravity Override Params {
  type=GravityOverrideParamsStruct
  default {
   emoteOrMelee {gravityOverride=<0 -150 -800>}
   onGround_Moving {gravityOverride=<0 -300 -800>}
   onGround_Standing {gravityOverride=<0 -150 -980>}
   skydive_GoingUp {gravityOverride=<0 -300 600>}
   tactical_Sprint {gravityOverride=<0 -450 0>}
  }
 }
 RBAN Params {
  type=RBANParamsStruct
  default {
   dBNO {simSpaceSettings {
     maxAngularAcceleration=3
     maxAngularVelocity=5
     maxLinearAcceleration=500
     maxLinearVelocity=1500
     velocityScaleZ=1
     externalLinearDragV=<1 1 0>
    }}
   dBNO_Carried {simSpaceSettings {maxAngularAcceleration=1, maxAngularVelocity=1, maxLinearAcceleration=10, maxLinearVelocity=0, velocityScaleZ=1, externalLinearDragV=<1 1 0>}}
   emoteOrMelee {
    componentLinearAccScale=<1 1 1>
    componentLinearVelScale=<1.1 1.1 1.1>
    simSpaceSettings {
     maxAngularAcceleration=3
     maxAngularVelocity=5
     maxLinearAcceleration=500
     maxLinearVelocity=1500
     velocityScaleZ=1
     externalLinearDragV=<1 1 0>
    }
   }
   front_End {simSpaceSettings {
     maxAngularAcceleration=1
     maxAngularVelocity=3.5
     maxLinearAcceleration=10
     maxLinearVelocity=500
     velocityScaleZ=1
     externalLinearDragV=<1 1 0>
    }}
   onGround_Crouch_Moving {simSpaceSettings {
     maxAngularAcceleration=3
     maxAngularVelocity=5
     maxLinearAcceleration=500
     maxLinearVelocity=1500
     velocityScaleZ=1
     externalLinearDragV=<1 1 0>
    }}
   onGround_Falling {simSpaceSettings {
     maxAngularAcceleration=3
     maxAngularVelocity=5
     maxLinearAcceleration=500
     maxLinearVelocity=1500
     velocityScaleZ=1
     externalLinearDragV=<1 1 0>
    }}
   onGround_Moving {
    componentLinearAccScale=<0.1 0.1 0.1>
    componentLinearVelScale=<0.2 0.2 0.2>
    simSpaceSettings {
     maxAngularAcceleration=3
     maxAngularVelocity=5
     maxLinearAcceleration=500
     maxLinearVelocity=1500
     velocityScaleZ=1
     externalLinearDragV=<1 1 0>
    }
   }
   onGround_Standing {
    componentLinearAccScale=<0.2 0.2 0.3>
    componentLinearVelScale=<0.3 0.3 0.4>
    simSpaceSettings {
     maxAngularAcceleration=50
     maxAngularVelocity=5
     maxLinearAcceleration=500
     maxLinearVelocity=1500
     velocityScaleZ=1
     externalLinearDragV=<1 1 0>
    }
   }
   rail_Grinding {simSpaceSettings {
     maxAngularAcceleration=3
     maxAngularVelocity=5
     maxLinearAcceleration=500
     maxLinearVelocity=1500
     velocityScaleZ=1
     worldAlpha=1
     externalLinearDragV=<1 1 0>
    }}
   riding_Moving {simSpaceSettings {
     maxAngularAcceleration=3
     maxAngularVelocity=5
     maxLinearAcceleration=500
     maxLinearVelocity=1500
     velocityScaleZ=1
     externalLinearDragV=<1 1 0>
    }}
   riding_Moving_Targeting {simSpaceSettings {
     maxAngularAcceleration=3
     maxAngularVelocity=5
     maxLinearAcceleration=500
     maxLinearVelocity=1500
     velocityScaleZ=1
     externalLinearDragV=<1 1 0>
    }}
   riding_Standing {simSpaceSettings {
     maxAngularAcceleration=3
     maxAngularVelocity=5
     maxLinearAcceleration=500
     maxLinearVelocity=1500
     velocityScaleZ=1
     externalLinearDragV=<1 1 0>
    }}
   skydive_DivingDown {
    componentLinearAccScale=<0 0 0>
    componentLinearVelScale=<0 0 0>
    simSpaceSettings {
     maxAngularAcceleration=3
     maxAngularVelocity=5
     maxLinearAcceleration=500
     maxLinearVelocity=1500
     velocityScaleZ=1
     externalLinearDragV=<1 1 0>
    }
   }
   skydive_FloatingDown {
    componentLinearAccScale=<0 0 0>
    componentLinearVelScale=<0 0 0>
    simSpaceSettings {
     maxAngularAcceleration=3
     maxAngularVelocity=5
     maxLinearAcceleration=500
     maxLinearVelocity=1500
     velocityScaleZ=1
     externalLinearDragV=<1 1 0>
    }
   }
   skydive_GoingUp {simSpaceSettings {
     maxAngularAcceleration=3
     maxAngularVelocity=5
     maxLinearAcceleration=500
     maxLinearVelocity=1500
     velocityScaleZ=1
     externalLinearDragV=<1 1 0>
    }}
   skydive_Parachuting {
    componentLinearAccScale=<0 0 0>
    componentLinearVelScale=<0 0 0>
    simSpaceSettings {
     maxAngularAcceleration=3
     maxAngularVelocity=5
     maxLinearAcceleration=500
     maxLinearVelocity=1500
     velocityScaleZ=1
     externalLinearDragV=<1 1 0>
    }
   }
   skydive_Surfing {simSpaceSettings {
     maxAngularAcceleration=3
     maxAngularVelocity=5
     maxLinearAcceleration=500
     maxLinearVelocity=1500
     velocityScaleZ=1
     externalLinearDragV=<1 1 0>
    }}
   swimming {simSpaceSettings {
     maxAngularAcceleration=3
     maxAngularVelocity=5
     maxLinearAcceleration=500
     maxLinearVelocity=1500
     velocityScaleZ=1
     externalLinearDragV=<1 1 0>
    }}
   tactical_Sprint {
    componentLinearAccScale=<0.2 0.2 0.2>
    componentLinearVelScale=<0.2 0.2 0.2>
    simSpaceSettings {
     maxAngularAcceleration=3
     maxAngularVelocity=5
     maxLinearAcceleration=500
     maxLinearVelocity=1500
     velocityScaleZ=1
     externalLinearDragV=<1 1 0>
    }
   }
   ziplining {simSpaceSettings {
     maxAngularAcceleration=0.5
     maxAngularVelocity=1
     maxLinearAcceleration=50
     maxLinearVelocity=200
     velocityScaleZ=1
     externalLinearDragV=<0.5 0.5 0>
    }}
  }
 }
 RBAN Settings Out {
  type=RBANSettingStruct
  default {
   componentLinearAccScale=<0 0 0>
   componentLinearVelScale=<0 0 0>
   simSpaceSettings {
    dampingAlpha=1
    maxAngularAcceleration=1000
    maxAngularVelocity=3
    maxLinearAcceleration=300
    maxLinearVelocity=700
    velocityScaleZ=0
    worldAlpha=1
    externalAngularVelocity=<0 0 0>
    externalLinearDragV=<0 0.5 0>
    externalLinearVelocity=<0 0 0>
   }
  }
 }
]
